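; Parent Blueprint for a construction paint pallet prop with selectable paint bucket materials, set up entirely in the construction script.
; An enum picks a BuildingTextureData entry, from which a dynamic material instance is built with diffuse, normal and specular textures.
; That material is applied to instanced bucket meshes and extra static mesh components.
; A random stream seeded from actor location randomizes each instance's yaw, keeping position, pitch, roll and scale.
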(event EventBeginPlay)

(event Collision|EventActorBeginOverlap
  (params (OtherActor :type Actor%20Object%20Reference)))

(event EventTick
  (params (DeltaSeconds :type Float%20%28single-precision%29)))

(fn ApplyRandomRotationToInstances
  (params (RandomStream :type Random%20Stream%20Structure%20%28by%20ref%29))
  (for _array_element (Variables|Default|GetAR_ISM_PaintBuckets)
    (bind _output_get (LocalVariables|SetL_ISM_Component _array_element))
    (Utilities|Array|Clear (LocalVariables|GetL_AR_InstanceTransforms))
    (bind _outtransforms (Components|InstancedStaticMesh|GetInstanceTransforms _output_get))
    (for _array_element_2 _outtransforms
      (Utilities|Array|Add (LocalVariables|GetL_AR_InstanceTransforms) (Math|Transform|MakeTransform (.location _array_element_2) (Math|Rotator|MakeRotator (.roll (.rotation _array_element_2)) (.pitch (.rotation _array_element_2)) (Math|Random|RandomFloatinRangefromStream RandomStream 0.0 3.6e+02)) (.scale _array_element_2))))
    (Components|InstancedStaticMesh|BatchUpdateInstancesTransforms (LocalVariables|GetL_ISM_Component) 0 (LocalVariables|GetL_AR_InstanceTransforms) false false false)))

(fn CreateMIDFromTextureData
  (params (MeshComponent :type Primitive%20Component%20Object%20Reference))
  (bind _l_fallbackmesh (LocalVariables|GetL_FallbackMesh))
  (bind _ar_texturedata (Variables|Default|GetAR_TextureData))
  (bind _l_texturedata (LocalVariables|GetL_TextureData))
  (bind _diffuse (Variables|/Script/FortniteGame.BuildingTextureData|GetDiffuse _l_texturedata))
  (bind _l_texturedata_2 (LocalVariables|GetL_TextureData))
  (bind _normal (Variables|/Script/FortniteGame.BuildingTextureData|GetNormal _l_texturedata_2))
  (bind _l_texturedata_3 (LocalVariables|GetL_TextureData))
  (bind _specular (Variables|/Script/FortniteGame.BuildingTextureData|GetSpecular _l_texturedata_3))
  (Utilities|IsValid MeshComponent
    (:Is%20Valid
      (LocalVariables|SetL_FallbackMesh MeshComponent))
    (:Is%20Not%20Valid
      _))
  (bind _validindex (CallFunction|IsValidIndex (Variables|Default|GetSelectedIndex) _ar_texturedata))
  (bind _output_get (LocalVariables|SetL_TextureData (Utilities|Array|Get(acopy) _ar_texturedata _validindex)))
  (Utilities|IsValid (Variables|/Script/FortniteGame.BuildingTextureData|GetOverrideMaterial _output_get)
    (:Is%20Valid
      (bind _returnvalue (Rendering|Material|CreateDynamicMaterialInstance (Variables|/Script/FortniteGame.BuildingTextureData|GetOverrideMaterial _output_get) "None" "None" :!class /Script/Engine.KismetMaterialLibrary))
      (LocalVariables|SetL_MID _returnvalue))
    (:Is%20Not%20Valid
      (Utilities|IsValid _l_fallbackmesh
        (:Is%20Valid
          (bind _returnvalue_2 (Rendering|Material|CreateDynamicMaterialInstance (Rendering|Material|GetMaterial _l_fallbackmesh 0) "None" "None" :!class /Script/Engine.KismetMaterialLibrary))
          (LocalVariables|SetL_MID _returnvalue_2))
        (:Is%20Not%20Valid
          (Development|PrintString "No material found.")
          (return)))))
  (Utilities|FlowControl|Sequence
    (:then_0
      (Utilities|IsValid _diffuse
        (:Is%20Valid
          (LocalVariables|SetL_Diffuse _diffuse)
          (Rendering|Material|SetTextureParameterValue (LocalVariables|GetL_MID) "Diffuse" (LocalVariables|GetL_Diffuse)))))
    (:then_1
      (Utilities|IsValid _normal
        (:Is%20Valid
          (LocalVariables|SetL_Normal _normal)
          (Rendering|Material|SetTextureParameterValue (LocalVariables|GetL_MID) "Normals" (LocalVariables|GetL_Normal)))))
    (:then_2
      (Utilities|IsValid _specular
        (:Is%20Valid
          (LocalVariables|SetL_Spec _specular)
          (Rendering|Material|SetTextureParameterValue (LocalVariables|GetL_MID) "SpecularMasks" (LocalVariables|GetL_Spec)))))
    (:then_3
      (Variables|Default|SetOverride%20Material (LocalVariables|GetL_MID))
      (return))))

(fn Get_ISM_Components
  (bind _returnvalue (Actor|GetComponentsByClass "/Script/Engine.InstancedStaticMeshComponent"))
  (bind _ar_ism_paintbuckets (Variables|Default|GetAR_ISM_PaintBuckets))
  (Utilities|Array|Clear (Variables|Default|GetAR_ISM_PaintBuckets))
  (if (> (Utilities|Array|Length _returnvalue) 0)
    (for _array_element _returnvalue
      (Utilities|IsValid (Variables|/Script/Engine.StaticMeshComponent|GetStaticMesh _array_element)
        (:Is%20Valid
          (Utilities|Array|AddUnique _ar_ism_paintbuckets _array_element)
          (if (> (Utilities|Array|Length _ar_ism_paintbuckets) 0)
            (Variables|Default|SetHas_ISM_Components true)
            (else
              (Variables|Default|SetHas_ISM_Components false))))
        (:Is%20Not%20Valid
          (Development|PrintString (Utilities|String|Append (.RightS (Utilities|String|Split (Utilities|GetDisplayName _array_element) ".")) " " "component needs a mesh assigned")))))
    (else
      (Variables|Default|SetHas_ISM_Components false))))

(fn GetAdditionalMeshComponents
  (bind _returnvalue (Actor|GetComponentsByClass "/Script/Engine.StaticMeshComponent"))
  (if (> (Utilities|Array|Length _returnvalue) 1)
    (bind _output_get (LocalVariables|SetL_AR_SM_Components _returnvalue))
    (Utilities|Array|RemoveItem _output_get (Variables|Debug|GetStaticMeshComponent))
    (return _output_get)))

(fn IsValidIndex
  (params (Index :type Integer) (AR_TextureData :type Array%20of%20Building%20Texture%20Data%20Object%20References))
  (bind _returnvalue (Utilities|Array|Length AR_TextureData))
  (bind _returnvalue_2 (- _returnvalue 1))
  (if (<= Index _returnvalue_2)
    (return Index)
    (else
      (return _returnvalue_2))))

(fn SelectPaintType
  (switch Utilities|FlowControl|Switch|SwitchonENUM_BucketMaterialOptions (Variables|Default|GetENUM_MaterialSelection)
    (:NewEnumerator0
      (Variables|Default|SetSelectedIndex 3))
    (:NewEnumerator1
      (Variables|Default|SetSelectedIndex 2))
    (:NewEnumerator2
      (Variables|Default|SetSelectedIndex 1))
    (:NewEnumerator3
      (Variables|Default|SetSelectedIndex 0)))
  (return))

(fn SetMaterialOn_ISM_Components
  (params (OverrideMaterial :type Material%20Interface%20Object%20Reference))
  (Utilities|IsValid OverrideMaterial
    (:Is%20Valid
      (if (Variables|Default|GetHas_ISM_Components)
        (for _array_element (Variables|Default|GetAR_ISM_PaintBuckets)
          (Rendering|Material|SetMaterial _array_element 0 OverrideMaterial))
        (else
          (Development|PrintString "No ISM components found"))))
    (:Is%20Not%20Valid
      (Development|PrintString "Missing override material."))))

(fn SetMaterialOnAddonComponents
  (params (OverrideMaterial :type Material%20Interface%20Object%20Reference))
  (Utilities|IsValid OverrideMaterial
    (:Is%20Valid
      (bind _ar_addoncomponents (CallFunction|GetAdditionalMeshComponents))
      (for _array_element _ar_addoncomponents
        (Rendering|Material|SetMaterial _array_element 0 OverrideMaterial)))))

(fn SetOverrideMaterial
  (params (MaterialIndex :type Integer) (AR_Materials :type Array%20of%20Material%20Interface%20Object%20References))
  (Variables|Default|SetBucketOverrideMaterial (Utilities|Array|Get(acopy) AR_Materials MaterialIndex)))

(fn SetRandomStream
  (params (distanceInterval :type Float%20%28double-precision%29) (includeZ :type Boolean))
  (bind _returnvalue (Math|Random|MakeRandomStream (Math|Float|Truncate (* (Class|EnvironmentArtFuncLib|RandomFloatbyLocation (Transformation|GetActorLocation) distanceInterval includeZ) 1e+04))))
  (return _returnvalue))

(fn ConstructionScript
  (bind _override_material (Variables|Default|GetOverride%20Material))
  (|CallParentFunction)
  (CallFunction|GetISMComponents)
  (CallFunction|SelectPaintType)
  (CallFunction|CreateMIDFromTextureData 0)
  (CallFunction|SetMaterialonISMComponents _override_material)
  (CallFunction|SetMaterialonAddonComponents _override_material)
  (if (Variables|Default|GetHas_ISM_Components)
    (bind _randomstream (CallFunction|SetRandomStream 96.0 true))
    (CallFunction|ApplyRandomRotationtoInstances _randomstream)))
