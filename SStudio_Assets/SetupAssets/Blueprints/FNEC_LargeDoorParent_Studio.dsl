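; Parent Blueprint for large openable doors, duplicated from FNEC_LargeDoorParent to avoid cross-plugin dependencies.
; Openable interface events set DoorState to transitional values; authority-only timeline completion sets end states and fires finish dispatchers.
; Activate/reset hooks flip bIsDoorOpen immediately and play audio offset by door progress; children supply the timeline.
; On authority, HandleTarget launches players along the door's right axis and applies a damage GameplayEffect to vehicles.
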(event EventBeginPlay
  ; This entire BP was duplicated from /SecretF_Assets/SetupAssets/Blueprints/FNEC_LargeDoorParent.FNEC_LargeDoorParent to avoid plugin dependency issues because these are in two different plugins.
  ;; IMPLEMENT / OVERRIDE IN CHILDREN
  (CallFunction|SetTimelineLength 0))

(event OpenableInterface|EventOpen
  (params (ControllerInstigator :type Controller%20Object%20Reference) (bRequestFastOpen :type Boolean))
  ; This entire BP was duplicated from /SecretF_Assets/SetupAssets/Blueprints/FNEC_LargeDoorParent.FNEC_LargeDoorParent to avoid plugin dependency issues because these are in two different plugins.
  ;;[0.2901 0.2942 0.7281 0.5] Openable Interface
  (if (Utilities|Enum|NotEqual(Enum) "NewEnumerator0" (Variables|Default|GetDoorState))
    (Variables|Default|SetDoorState "NewEnumerator1")))

(event OpenableInterface|EventClose
  (params (ControllerInstigator :type Controller%20Object%20Reference) (bRequestFastClose :type Boolean))
  ; This entire BP was duplicated from /SecretF_Assets/SetupAssets/Blueprints/FNEC_LargeDoorParent.FNEC_LargeDoorParent to avoid plugin dependency issues because these are in two different plugins.
  ;;[0.2901 0.2942 0.7281 0.5] Openable Interface
  (if (Utilities|Enum|NotEqual(Enum) (Variables|Default|GetDoorState) "NewEnumerator2")
    (Variables|Default|SetDoorState "NewEnumerator3")))

(fn BP_OnActive
  (params (EnableAlternateActive :type Boolean) (EnableOpenCues :type Boolean))
  (Variables|Default|SetbIsDoorOpen true) ; As soon as we start opening the door, the Openable\ninterface should say it's open.
  (if EnableOpenCues
    (CallFunction|PlayDoorAudio true)))

(fn BP_OnClosed
  (params (EnableAlternateClose :type Boolean) (EnableOpenCues :type Boolean)))

(fn BP_OnOpen
  (params (EnableAlternateOpen :type Boolean) (EnableOpenCues :type Boolean)))

(fn BP_OnReset
  (params (EnableAlternateReset :type Boolean) (EnableOpenCues :type Boolean))
  (Variables|Default|SetbIsDoorOpen false) ; As soon as we start closing the door, the Openable\ninterface should say it's closed.
  (if EnableOpenCues
    (CallFunction|PlayDoorAudio false)))

(fn Close_Internal_OpenableInterface)

(fn GetComponentToLock
  (return 0))

(fn HandleOpenClose_OpenableInterface
  (params (bOpen? :type Boolean) (Instigator :type Controller%20Object%20Reference) (bFastOpen? :type Boolean))
  (if bOpen?
    (OpenableInterfaceSupport|OpenInternalOpenableInterface)
    (else
      (OpenableInterfaceSupport|CloseInternalOpenableInterface))))

(fn HandleTarget
  (params (TargetedActor :type Actor%20Object%20Reference))
  (Networking|SwitchHasAuthority
    (:Authority
      (bind _asfort_player_pawn (Utilities|Casting|CastToFortPlayerPawn (Variables|GetTargetedActor))
        (:then
          (Character|LaunchCharacter _asfort_player_pawn (* (ScalableFloat|GetValueatLevel (Variables|Default|GetHF_PushAwayStrength)) (Math|Vector|SelectVector (Math|Vector|NegateVector (Transformation|GetActorRightVector)) (Transformation|GetActorRightVector) (> (Math|Vector|DotProduct (Transformation|GetActorRightVector) "0.000000,0.000000,1.000000") 0))) false false))
        (:CastFailed
          (bind _asfort_athena_vehicle (Utilities|Casting|CastToFortAthenaVehicle (Variables|GetTargetedActor)))
          (CallFunction|SmashVehicle _asfort_athena_vehicle))))))

(fn IsOpen
  (return (Variables|Default|GetbIsDoorOpen)))

(fn OnDoorTimelineFinished
  (params (Direction :type ETimelineDirection%20Enum :default "Forward"))
  (Networking|SwitchHasAuthority
    (:Authority
      (switch Utilities|FlowControl|Switch|SwitchonETimelineDirection Direction
        (:Forward
          (Variables|Default|SetDoorState "NewEnumerator0")
          (Default|CallOnDoorFinishActivation))
        (:Backward
          (Variables|Default|SetDoorState "NewEnumerator2")
          (Default|CallOnDoorFinishReset))))))

(fn Open_Internal_OpenableInterface)

(fn PlayDoorAudio
  (params (bOpen :type Boolean))
  (bind _dooraudio (Variables|Default|GetDoorAudio))
  (bind _dooraudio_2 (Variables|Default|GetDoorAudio))
  (bind _timelinelength (Variables|Default|GetTimelineLength))
  (if bOpen
    (Audio|Components|Audio|SetSound _dooraudio_2 (Variables|Default|GetOpenSound))
    (Audio|Components|Audio|Play _dooraudio_2 (* (Variables|Default|GetDoorOperationPercentage) _timelinelength))
    (else
      (Audio|Components|Audio|SetSound _dooraudio (Variables|Default|GetCloseSound))
      (Audio|Components|Audio|Play _dooraudio (* (- 1.0 (Variables|Default|GetDoorOperationPercentage)) _timelinelength)))))

(fn SetTimelineLength
  (params (InTimeline :type Timeline%20Component%20Object%20Reference))
  ; IMPLEMENT / OVERRIDE IN CHILDREN
  (Variables|Default|SetTimelineLength (Components|Timeline|GetTimelineLength InTimeline)))

(fn SmashVehicle
  (params (Vehicle :type Fort%20Athena%20Vehicle%20Object%20Reference))
  (Fortnite|ApplyGameplayEffectsToActor Vehicle (Utilities|Array|MakeArray (Variables|Default|GetVehicleDamageGE)) (Transformation|GetActorLocation) false 0 (Ability|GetAbilitySystemComponent 0))
  (return))

(fn ConstructionScript)
